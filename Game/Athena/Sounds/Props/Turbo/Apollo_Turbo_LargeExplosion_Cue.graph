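# Sound cue for the large explosion sound of the Apollo Turbo prop.
# A single wave player feeds the root with no randomization or modulation, at slightly reduced volume.
# It uses the shared close-explosion attenuation asset and the impacts sound class.
# A concurrency override caps it at one playing instance per owner.
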SoundCue /Game/Athena/Sounds/Props/Turbo/Apollo_Turbo_LargeExplosion_Cue
root=WavePlayer_0
nodes [{id=WavePlayer_0, properties {soundWaveAssetPtr=&/Apollo_Turbo_LargeExplosion}}]
properties {
 attenuationSettings=/Game/Sounds/Attenuation/Fort_Explosion_Close_Athena_Att
 bOverrideConcurrency=true
 maxDistance=1499.9875
 soundClassObject=&Fort_Impacts
 volumeMultiplier=0.8
 concurrencyOverrides {bLimitToOwner=true, maxCount=1}
}
