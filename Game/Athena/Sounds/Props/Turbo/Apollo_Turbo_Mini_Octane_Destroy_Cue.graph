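# Sound cue for the destruction of the Apollo Turbo Mini Octane prop in Athena.
# A single wave player plays one destroy wave with no randomization or mixing, at a boosted volume.
# It uses the Fort_Impacts sound class and a close-explosion attenuation asset.
# A concurrency override limits it to one playing instance per owner.
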SoundCue /Game/Athena/Sounds/Props/Turbo/Apollo_Turbo_Mini_Octane_Destroy_Cue
root=WavePlayer_1
nodes [{id=WavePlayer_1, properties {soundWaveAssetPtr=&/Prop_Turbo_Mini_Octane_Destroy_01}}]
properties {
 attenuationSettings=/Game/Sounds/Attenuation/Fort_Explosion_Close_Athena_Att
 bOverrideConcurrency=true
 maxDistance=1499.9875
 soundClassObject=&Fort_Impacts
 volumeMultiplier=1.8
 concurrencyOverrides {bLimitToOwner=true, maxCount=1}
}
