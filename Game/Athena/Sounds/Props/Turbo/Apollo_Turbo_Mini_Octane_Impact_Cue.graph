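# Sound cue for the impact sound of the Turbo Mini Octane prop in Athena.
# A Random node picks between two impact wave variations with equal weight, played at slightly reduced volume in the Fort_Impacts sound class.
# It uses a close explosion attenuation asset, and a concurrency override limits it to one playing instance per owner.
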SoundCue /Game/Athena/Sounds/Props/Turbo/Apollo_Turbo_Mini_Octane_Impact_Cue
root=Random_0
nodes [
 {id=Random_0, children [WavePlayer_1,WavePlayer_2], properties {weights [1,1]}}
 {id=WavePlayer_1, properties {soundWaveAssetPtr=&/Prop_Turbo_Mini_Octane_Impact_01}}
 {id=WavePlayer_2, properties {soundWaveAssetPtr=&/Prop_Turbo_Mini_Octane_Impact_02}}
]
properties {
 attenuationSettings=/Game/Sounds/Attenuation/Fort_Explosion_Close_Athena_Att
 bOverrideConcurrency=true
 maxDistance=1499.9875
 soundClassObject=&Fort_Impacts
 volumeMultiplier=0.8
 concurrencyOverrides {bLimitToOwner=true, maxCount=1}
}
