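# Sound cue for Apollo Turbo prop music, routed to the NPC music sound class, with no nodes or root, so it references no audio as stored.
# Its own attenuation override defines a short-range falloff with wide stereo spread, occlusion low-pass filtering and interpolated listener focus that scales non-focus volume to zero.
# Its concurrency override limits playback to one instance per owner.
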SoundCue /Game/Athena/Sounds/Props/Turbo/Apollo_Turbo_Music_Cue
root=null
nodes []
properties {
 bOverrideAttenuation=true
 bOverrideConcurrency=true
 maxDistance=768
 soundClassObject=&Fort_Music_NPC
 volumeMultiplier=1
 attenuationOverrides {
  bEnableFocusInterpolation=true
  bEnableListenerFocus=true
  bEnableOcclusion=true
  falloffDistance=512
  focusAttackInterpSpeed=5
  nonFocusVolumeAttenuation=0
  occlusionInterpolationTime=0.25
  occlusionLowPassFilterFrequency=1000
  occlusionTraceChannel=ECC_GameTraceChannel11
  stereoSpread=2048
  attenuationShapeExtents {x=256}
 }
 concurrencyOverrides {bLimitToOwner=true, maxCount=1}
}
